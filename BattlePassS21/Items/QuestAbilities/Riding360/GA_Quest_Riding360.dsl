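; Quest gameplay ability for the Riding360 challenge that credits a roughly full 360-degree spin while the player pawn is falling.
; When the MovementMode.Falling tag is added, it records the pawn's rotation and starts a looping 0.1-second timer.
; Each tick accumulates the angle between previous and current forward vectors.
; When falling ends, it stops the timer, then sends the quest stat event if the total nears 360, otherwise resets it.
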(event Ability|EventActivateAbilityFromEvent
  (params (EventData :type Gameplay%20Event%20Data%20Structure%20%28by%20ref%29))
  (|CallParentFunction EventData)
  (if (GameplayTags|HasMatchingGameplayTag (Variables|GAVars|Data|GetOwningPlayerPawn) "(TagName=\"MovementMode.Falling\")")
    (Ability|Tasks|WaitGameplayTagRemove "(TagName=\"MovementMode.Falling\")" (Variables|GAVars|Data|GetOwningPlayerPawn)
      (:Removed
        _)))
  (Ability|Tasks|WaitGameplayTagAdd "(TagName=\"MovementMode.Falling\")" (Variables|GAVars|Data|GetOwningPlayerPawn)
    (:Added
      (Variables|Default|SetInitial%20Rotation (Transformation|GetActorRotation (Variables|GAVars|Data|GetOwningPlayerPawn)))
      (bind _returnvalue (Utilities|Time|SetTimerbyEvent (delegate Custom|Tick%20Angle%20Check) 0.1 true))
      (Variables|Default|SetTimer%20Handle _returnvalue)
      (Ability|Tasks|WaitGameplayTagRemove "(TagName=\"MovementMode.Falling\")" (Variables|GAVars|Data|GetOwningPlayerPawn)
        (:Removed
          (Utilities|Time|ClearandInvalidateTimerbyHandle (Variables|Default|GetTimer%20Handle))
          (if (>= (Math|Float|Absolute(Float) (Variables|Default|GetDegrees%20Turned)) 3.5e+02)
            (Quest|SendComplexCustomStatEvent (Variables|GAVars|Data|GetOwningPlayerQuestManager) 0 "(GameplayTags=((TagName=\"Athena.Quests.Riding360\")))" "(GameplayTags=)" 1)
            (else
              (Variables|Default|SetDegrees%20Turned 0.0))))))))

(event Custom|Tick%20Angle%20Check
  (bind _owningplayerpawn (Variables|GAVars|Data|GetOwningPlayerPawn))
  (bind _returnvalue (Transformation|GetActorRotation _owningplayerpawn))
  (Variables|Default|SetDegrees%20Turned (+ (Class|AbilityFunctions|GetAngleBetweenTwoVectors (Math|Vector|GetForwardVector _returnvalue) (Math|Vector|GetForwardVector (Variables|Default|GetInitial%20Rotation))) (Variables|Default|GetDegrees%20Turned)))
  (Variables|Default|SetInitial%20Rotation _returnvalue))
